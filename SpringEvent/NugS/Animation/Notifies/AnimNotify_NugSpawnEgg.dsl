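; Anim notify that makes a SpringEvent NugS chicken lay an egg, spawning it only on the non-cosmetic (server) branch.
; It checks nav projection, traces down, then spawns a tilted, tagged egg: golden by capped hotfixable chance outside Creative, otherwise Heal or Hop by chicken variant.
; Regular eggs join the chicken's laid list; exceeding the total cap flags the oldest for overflow despawn and removes it, while golden eggs only increment a counter.
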(fn CanLayGoldEgg
  (params (Owner :type Actor%20Object%20Reference))
  ;[1 1 1] FIX FORT-575928: don't lay golden eggs in Creative / UEFN as we don't have gold bars there
  (if (Fortnite|IsCreativeorPlaygroundModePlaylist)
    (return false)
    (else
      (return (and (< (Variables|/SpringEvent/NugS/Pawns/NPC_Pawn_Irwin_Prey_NugS_Base.NPC_Pawn_Irwin_Prey_NugS_Base_C|GetGoldEggLayed (Utilities|Casting|CastToNPC_Pawn_Irwin_Prey_NugS_Base Owner)) (ScalableFloat|GetValueatLevel (Variables|Default|GetMaxGoldEggAmountHotfix) 1.0)) (<= (Math|Random|RandomFloatinRange 0.0 1.0) (ScalableFloat|GetValueatLevel (Variables|Default|GetGoldEggChance) 1.0)))))))

(fn CompareTotalEggAmount
  (params (SpringChicken :type NPC%20Pawn%20Irwin%20Prey%20Nug%20S%20Base%20Object%20Reference))
  (return (<= (Utilities|Array|Length (Variables|/SpringEvent/NugS/Pawns/NPC_Pawn_Irwin_Prey_NugS_Base.NPC_Pawn_Irwin_Prey_NugS_Base_C|GetLayedEggs SpringChicken)) (Math|Float|Truncate (ScalableFloat|GetValueatLevel (Variables|Default|GetMaxTotalEggAmount))))))

(fn DestroyOldestEgg
  (params (SpringChicken :type NPC%20Pawn%20Irwin%20Prey%20Nug%20S%20Base%20Object%20Reference))
  (bind _output (Utilities|Array|Get(acopy) (Variables|/SpringEvent/NugS/Pawns/NPC_Pawn_Irwin_Prey_NugS_Base.NPC_Pawn_Irwin_Prey_NugS_Base_C|GetLayedEggs SpringChicken)))
  (bind _layedeggs (Variables|/SpringEvent/NugS/Pawns/NPC_Pawn_Irwin_Prey_NugS_Base.NPC_Pawn_Irwin_Prey_NugS_Base_C|GetLayedEggs SpringChicken))
  (if (Utilities|IsArrayNonEmpty _layedeggs)
    (Utilities|IsValid _output
      (:Is%20Valid
        (bind _ascbga_gathered_spring_egg_hatch_parent (Utilities|Casting|CastToCBGA_Gathered_SpringEgg_Hatch_Parent _output)
          (:then
            (Variables|/SpringEvent/Items/CBGA_Gathered_SpringEgg_Hatch_Parent.CBGA_Gathered_SpringEgg_Hatch_Parent_C|SetEggOverflowDespawn true _ascbga_gathered_spring_egg_hatch_parent))
          (:CastFailed
            _)))
      (:Is%20Not%20Valid
        _))
    (Utilities|Array|RemoveIndex _layedeggs)))

(fn Received_Notify
  (params (MeshComp :type Skeletal%20Mesh%20Component%20Object%20Reference) (Animation :type Anim%20Sequence%20Base%20Object%20Reference) (EventReference :type Anim%20Notify%20Event%20Reference%20Structure%20%28by%20ref%29))
  (bind _returnvalue (Components|GetOwner MeshComp))
  (bind _returnvalue_2 (Transformation|GetActorLocation _returnvalue))
  (bind (_projectedlocation _returnvalue_3) (AI|Navigation|ProjectPointtoNavigation _returnvalue_2 0 "/Game/Athena/AI/Phoebe/NavFilter/NavFilter_Phoebe_NoWater.NavFilter_Phoebe_NoWater_C" "0.000000,0.000000,-25.000000"))
  (bind _returnvalue_4 (Math|Random|RandomFloatinRange -1e+01 1e+01))
  (bind _returnvalue_5 (Math|Random|RandomFloatinRange 0.0 3.6e+02))
  (bind _returnvalue_6 (Math|Random|RandomFloatinRange -1e+01 1e+01))
  (bind _asnpc_pawn_irwin_prey_nug_s_base (Utilities|Casting|CastToNPC_Pawn_Irwin_Prey_NugS_Base _returnvalue)
    (:then
      (Networking|CanExecuteCosmeticEvents
        (:True
          (Development|PrintString "Client:: SpawnEggDrop"))
        (:False
          ;[1 1 1] Spawn Egg (Only on Server)
          (if _returnvalue_3
            (bind (_outhit _returnvalue_7) (Collision|LineTraceForObjects _returnvalue_2 (- _returnvalue_2 "0.000000,0.000000,1000.000000") (Utilities|Array|MakeArray "ObjectTypeQuery1")))
            (if _returnvalue_7
              (if (CallFunction|CanLayGoldEgg _returnvalue)
                (bind _returnvalue_8 (Game|SpawnActorfromClass "/SpringEvent/Items/GoldenEgg/CBGA_SpringEgg_Golden.CBGA_SpringEgg_Golden_C" (Math|Transform|MakeTransform (+ (.Location (Collision|BreakHitResult _outhit)) "0.000000,0.000000,15.000000") (Math|Rotator|MakeRotfromZ (Math|Vector|RotateVector (.ImpactNormal (Collision|BreakHitResult _outhit)) (Math|Rotator|MakeRotator _returnvalue_4 _returnvalue_6 _returnvalue_5)))) "Undefined" "OverrideRootScale" 0 0 0))
                (CallFunction|SetLaidbyChicken _returnvalue_8)
                (Math|Integer|IncrementInt (Variables|/SpringEvent/NugS/Pawns/NPC_Pawn_Irwin_Prey_NugS_Base.NPC_Pawn_Irwin_Prey_NugS_Base_C|GetGoldEggLayed _asnpc_pawn_irwin_prey_nug_s_base)
                  (:%20%20
                    _))
                (else
                  (bind _returnvalue_9 (Game|SpawnActorfromClass (Utilities|Select :Index (Variables|/SpringEvent/NugS/Pawns/NPC_Pawn_Irwin_Prey_NugS_Base.NPC_Pawn_Irwin_Prey_NugS_Base_C|GetChosenVariant _asnpc_pawn_irwin_prey_nug_s_base) :NewEnumerator0 "/SpringEvent/Items/HealEgg/CBGA_Gathered_SpringEgg_Heal.CBGA_Gathered_SpringEgg_Heal_C" :NewEnumerator1 "/SpringEvent/Items/HopEgg/CBGA_SpringEgg_Hops.CBGA_SpringEgg_Hops_C") (Math|Transform|MakeTransform (+ (.Location (Collision|BreakHitResult _outhit)) "0.000000,0.000000,15.000000") (Math|Rotator|MakeRotfromZ (Math|Vector|RotateVector (.ImpactNormal (Collision|BreakHitResult _outhit)) (Math|Rotator|MakeRotator _returnvalue_4 _returnvalue_6 _returnvalue_5)))) "Undefined" "SelectDefaultAtRuntime" 0 0))
                  (CallFunction|SetLaidbyChicken _returnvalue_9)
                  (Utilities|Array|Add (Variables|/SpringEvent/NugS/Pawns/NPC_Pawn_Irwin_Prey_NugS_Base.NPC_Pawn_Irwin_Prey_NugS_Base_C|GetLayedEggs _asnpc_pawn_irwin_prey_nug_s_base) _returnvalue_9)))
              (if (not (CallFunction|CompareTotalEggAmount _asnpc_pawn_irwin_prey_nug_s_base))
                (CallFunction|DestroyOldestEgg _asnpc_pawn_irwin_prey_nug_s_base))
              (Development|PrintString "Server:: SpawnEggDrop")
              (return false)
              (else
                (Development|PrintString "Server:: NoLocationFound")
                (return false)))
            (else
              _)))))
    (:CastFailed
      _))
  (return false))

(fn SetLaidByChicken
  (params (Object :type Actor%20Object%20Reference))
  (Ability|GameplayEffect|AddLooseGameplayTags Object (LocalVariables|GetLaid%20Gameplay%20Tag) true))
